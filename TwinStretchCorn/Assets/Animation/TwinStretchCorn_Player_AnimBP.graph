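# Player item-layer AnimBlueprint for TwinStretchCorn on the Fortnite player skeleton, handling weapon fire and shield poses.
# Only ItemFinalPoseOverride holds real logic, using bool blends over a cached input pose to choose fire or shield branches.
# Fire feeds alternating Fire/ReFire states into an aim offset.
# Shield layers its aim offset from Pelvis excluding the neck, or runs a shield-broken unequip animation that preserves the lower body through a blend mask.
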AnimBlueprint /TwinStretchCorn/Assets/Animation/TwinStretchCorn_Player_AnimBP
target_skeleton=&Fortnite_M_Avg_Player_Skeleton
graphs [
 AnimGraph {nodes [Output Pose()]}
 EventGraph {links [], nodes []}
 ItemAimOffset {nodes [
   Input Pose!() -> (Pose, AimOffsetAlpha:double, Yaw:double, Pitch:double) {name=InPose_AimOffset}
   Input Pose!() -> (Pose) {name=InPose_PreAimOffsets}
   Output Pose() {layerGroup=Item Override}
  ]}
 ItemFallAdditive {nodes [
   Input Pose() {name=InPoseItemFallAdditive}
   Output Pose() {layerGroup=Item Override}
  ]}
 ItemFallLandAdditive {nodes [
   Input Pose() {name=InPoseItemFallLandAdditive}
   Output Pose() {layerGroup=Item Override}
  ]}
 ItemFallLevel2Override {nodes [
   Input Pose() {name=InPose_FallLevel2}
   Output Pose() {layerGroup=Item Override}
  ]}
 ItemFinalPoseOverride {nodes [
   Blend Poses by bool(
    BlendPose_0=Sequence Player 'Falling_Idle'() {bLoopAnimation=false, sequence=/TwinStretchCorn/Assets/Animation/New_Animations/Falling_Idle}
    BlendTime_0=0.5
    BlendTime_1=0.2
   ) {blendType=Linear}
   Blend Poses by bool(
    BlendPose_0=Sequence Player 'Falling_Idle'() {bLoopAnimation=false, sequence=&Falling_Idle}
    BlendTime_0=0.5
    BlendTime_1=0.2
   ) {blendType=Linear}
   Blend Poses by bool(
    BlendPose_0=Sequence Player 'Falling_Idle'() {bLoopAnimation=false, sequence=&Falling_Idle}
    BlendTime_0=0.5
    BlendTime_1=0.2
   ) {blendType=Linear}
   n1 = AimOffset Player 'TwinStretchCorn_Shield_AO'!(
    X:float=Get!() -> (AdjustedAim_ResultingYaw:float=~, AdjustedAim_ResultingPitch:float=~, AdjustedAim_ZeroOutPitchWeightCurveName:Name)
    Y:float=Get ClampedPitch() -> (ClampedPitch:float)
    BasePose:Pose Link Structure=Use cached pose 'Input Pose'()
    Alpha:float=1
   ) -> (Pose:Pose Link Structure) {
    blendSpace=/TwinStretchCorn/Assets/Animation/New_Animations/Shards_WEP/Shield_AO_Targets/TwinStretchCorn_Shield_AO
    alphaBoolBlend {blendOption=Linear}
   }
   n2 = Layered blend per bone(BasePose=Use cached pose 'Input Pose'(), BlendPoses_0=@n1) {bMeshSpaceRotationBlend=true, layerSetup [{branchFilters [{boneName=Pelvis}, {blendDepth=-1, boneName=Neck_01}]}]}
   AnimGraphNode_StateMachine_1 = ShieldBroken SM() {is=title AnimGraphNode_StateMachine}
   AnimGraphNode_StateMachine_0 = Fire SM() {is=title AnimGraphNode_StateMachine}
   n3 = Use cached pose 'Input Pose'()
   n4 = AimOffset Player 'TwinStretchCorn_Fire_AO'!(
    X:float=0
    Y:float=Get() -> (AdjustedAim_ResultingYaw=~, AdjustedAim_ResultingPitch=~).AdjustedAim_ResultingPitch
    BasePose:Pose Link Structure=Blend Poses by bool(
     BlendPose_0=Use cached pose 'Input Pose'()
     BlendPose_1=Two Way Blend(A=@n3, B=@AnimGraphNode_StateMachine_0, Alpha=0) {blendNode {alphaBoolBlend {blendOption=Linear}}}
     BlendTime_0=0.2
     BlendTime_1=0.2
    ) {blendType=Linear}
    Alpha:float=1
   ) -> (Pose:Pose Link Structure) {blendSpace=/TwinStretchCorn/Assets/Animation/New_Animations/Fire_AO_Targets/TwinStretchCorn_Fire_AO, alphaBoolBlend {blendOption=Linear}}
   Output Pose(Result=Blend Poses by bool(
     BlendPose_0=@n4
     BlendPose_1=Blend Poses by bool(BlendPose_0=@n2, BlendPose_1=@AnimGraphNode_StateMachine_1, BlendTime_0=0.2, BlendTime_1=0.2) {blendType=Linear}
     BlendTime_0=0.2
     BlendTime_1=0.2
    ) {blendType=Linear}) {layerGroup=Item Override}
   Save cached pose 'Input Pose'(Pose=Input Pose() {name=InFinalPose}) {cacheName=Input Pose}
  ]}
 Fire SM {
  entry=Fire
  path=ItemFinalPoseOverride.AnimGraphNode_StateMachine_0.Fire SM
  ids=id
  nodes [
   To Fire() {is=title AnimStateAliasNode, also {aliases [ReFire]}}
   To ReFire() {is=title AnimStateAliasNode, also {aliases [Fire]}}
   AnimStateNode_0 = ReFire() {is=title AnimStateNode}
   AnimStateNode_2 = Fire() {is=title AnimStateNode}
   AnimStateTransitionNode_2 = AnimStateTransitionNode() {crossfadeDuration=0.25}
   AnimStateTransitionNode_3 = AnimStateTransitionNode() {crossfadeDuration=0.25}
  ]
  links [
   {
    dst=Fire
    rule=ItemFinalPoseOverride.AnimGraphNode_StateMachine_0.Fire SM.AnimStateTransitionNode_2.Transition
    src=To Fire
    via=AnimStateTransitionNode_2
   }
   {
    dst=ReFire
    rule=ItemFinalPoseOverride.AnimGraphNode_StateMachine_0.Fire SM.AnimStateTransitionNode_3.Transition
    src=To ReFire
    via=AnimStateTransitionNode_3
   }
  ]
 }
 ReFire {
  path=ItemFinalPoseOverride.AnimGraphNode_StateMachine_0.Fire SM.AnimStateNode_0.ReFire
  nodes [Output Animation Pose(Result=Sequence Player 'Fire_New'() {bLoopAnimation=false, sequence=/TwinStretchCorn/Assets/Animation/New_Animations/Fire_New})]
 }
 Fire {
  path=ItemFinalPoseOverride.AnimGraphNode_StateMachine_0.Fire SM.AnimStateNode_2.Fire
  nodes [Output Animation Pose(Result=Sequence Player 'Fire_New'() {bLoopAnimation=false, sequence=&Fire_New})]
 }
 Transition {
  path=ItemFinalPoseOverride.AnimGraphNode_StateMachine_0.Fire SM.AnimStateTransitionNode_2.Transition
  nodes [Result(bCanEnterTransition=Get bIsFiring() -> (bIsFiring:bool))]
 }
 Transition {
  path=ItemFinalPoseOverride.AnimGraphNode_StateMachine_0.Fire SM.AnimStateTransitionNode_3.Transition
  nodes [Result(bCanEnterTransition=Get bIsReFiring() -> (bIsReFiring:bool))]
 }
 ShieldBroken SM {
  entry=PassThrough
  path=ItemFinalPoseOverride.AnimGraphNode_StateMachine_1.ShieldBroken SM
  ids=id
  nodes [
   AnimStateNode_0 = PassThrough() {is=title AnimStateNode}
   AnimStateNode_1 = State() {is=title AnimStateNode}
   AnimStateTransitionNode_0 = AnimStateTransitionNode() {crossfadeDuration=0.1}
   AnimStateTransitionNode_1 = AnimStateTransitionNode() {bAutomaticRuleBasedOnSequencePlayerInState=true}
  ]
  links [
   {
    dst=State
    rule=ItemFinalPoseOverride.AnimGraphNode_StateMachine_1.ShieldBroken SM.AnimStateTransitionNode_0.Transition
    src=PassThrough
    via=AnimStateTransitionNode_0
   }
   {
    dst=PassThrough
    rule=ItemFinalPoseOverride.AnimGraphNode_StateMachine_1.ShieldBroken SM.AnimStateTransitionNode_1.Transition
    src=State
    via=AnimStateTransitionNode_1
   }
  ]
 }
 PassThrough {
  path=ItemFinalPoseOverride.AnimGraphNode_StateMachine_1.ShieldBroken SM.AnimStateNode_0.PassThrough
  nodes [Output Animation Pose(Result=Use cached pose 'Input Pose'())]
 }
 State {
  path=ItemFinalPoseOverride.AnimGraphNode_StateMachine_1.ShieldBroken SM.AnimStateNode_1.State
  nodes [
   n1 = Layered blend per bone(
    BasePose=Sequence Player 'Alt_Unequip_Shield'() {bLoopAnimation=false, sequence=/TwinStretchCorn/Assets/Animation/Shield_Animations/Alt_Unequip_Shield}
    BlendPoses_0=Use cached pose 'Input Pose'()
   ) {blendMode=BlendMask, blendMasks [&Fortnite_M_Avg_Player_Skeleton:LowerBody], layerSetup []}
   Output Animation Pose(Result=@n1)
  ]
 }
 Transition {
  path=ItemFinalPoseOverride.AnimGraphNode_StateMachine_1.ShieldBroken SM.AnimStateTransitionNode_0.Transition
  nodes [Result(bCanEnterTransition=Get bIsShieldBroken() -> (bIsShieldBroken:bool))]
 }
 Transition {path=ItemFinalPoseOverride.AnimGraphNode_StateMachine_1.ShieldBroken SM.AnimStateTransitionNode_1.Transition, nodes [Result()]}
 ItemFullBodyAimOffsets {nodes [
   Input Pose!() -> (Pose, FullBody_AimOffsetAlpha:double, Aim_Yaw:double, Aim_Pitch:double) {name=InPoseFullBodyAimOffsets}
   Input Pose!() -> (Pose) {name=InPosePreFullBodyAimOffsets}
   Output Pose() {layerGroup=Item Override}
  ]}
 ItemFullBodyOverride {nodes [
   Input Pose() {name=InPoseFullBody}
   Output Pose() {layerGroup=Item Override}
  ]}
 ItemIdleOverride {nodes [
   Input Pose() {name=InPose_Idle}
   Output Pose() {layerGroup=Item Override}
  ]}
 ItemJetPackJumpAdditive {nodes [Output Pose() {layerGroup=Item Override}]}
 ItemJetPackStartAdditive {nodes [Output Pose() {layerGroup=Item Override}]}
 ItemJumpLoopAdditive {nodes [
   Input Pose() {name=InPoseItemJumpLoopAdditive}
   Output Pose() {layerGroup=Item Override}
  ]}
 ItemJumpOverride {nodes [
   Input Pose() {name=InPoseJump}
   Output Pose() {layerGroup=Item Override}
  ]}
 ItemJumpUpAdditive {nodes [
   Input Pose() {name=InPoseItemJumpUpAdditive}
   Output Pose() {layerGroup=Item Override}
  ]}
 ItemPreAimOffsets {nodes [
   Input Pose() {name=InPosePreAimOffsets}
   Output Pose() {layerGroup=Item Override}
  ]}
 ItemPreIK {nodes [
   Input Pose() {name=InPose_PreIK}
   Output Pose() {layerGroup=Item Override}
  ]}
 ItemRightHandIK {nodes [
   Input Pose() {name=InPosePreRightHandIK}
   Input Pose() {name=InPosePostRightHandIK}
   Output Pose() {layerGroup=Item Override}
  ]}
 ItemSkeletalControl {nodes [
   Input Pose()
   Output Pose() {layerGroup=Item Override}
  ]}
 ItemSwimJumpFallLoop {nodes [Output Pose() {layerGroup=Item Override}]}
 ItemSwimJumpSurfaceEnd {nodes [Output Pose() {layerGroup=Item Override}]}
 ItemSwimJumpSurfaceLoop {nodes [Output Pose() {layerGroup=Item Override}]}
 ItemUpperBody {nodes [
   Input Pose!() -> (Pose, InputParam:Fort Anim Input Adjusted Aim Structure) {name=InPoseUpperBody}
   Output Pose() {layerGroup=Item Override}
  ]}
 ItemZipLineStartAdditive {nodes [Output Pose() {layerGroup=Item Override}]}
]
variables []
