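# Animation Blueprint for the TwinStretchCorn shield gadget, showing the animated shield while blocking and otherwise collapsing it via zero root scale.
# A Shield state machine drives equip, idle, hit reaction, unequip and broken poses from bool flags.
# Three lag blendspaces and a flag-triggered idle-noise state machine layer on additively, followed by a root yaw rotation.
# Root translation nodes reading ShieldPosOffset sit unconnected to the output.
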AnimBlueprint /TwinStretchCorn/Assets/Animation/TwinStretchCorn_Shield_AnimBP
target_skeleton=/TwinStretchCorn/Meshes/TwinStretchCorn_Gadget_Skeleton
graphs [
 AnimGraph {nodes [
   Apply Additive(Additive="Sequence Player (additive) 'TwinStretchCorn_Shield_Idle_Additive_WEP'"() {sequence=/TwinStretchCorn/Assets/Animation/New_Animations/Shards_WEP/TwinStretchCorn_Shield_Idle_Additive_WEP}) {alphaBoolBlend {blendOption=Linear}}
   "Transform (Modify) Bone - Bone: root"!(
    ComponentPose=Local To Component()
    Alpha
    Translation_X:double=0
    Translation_Y:double=0
    Translation_Z:double=Get ShieldPosOffset() -> (ShieldPosOffset:float)
    Rotation="0,0,0"
    Scale="0,0,0"
   ) -> (Pose) {
    rotationSpace=BCS_BoneSpace
    translationMode=BMM_Additive
    translationSpace=BCS_BoneSpace
    alphaBoolBlend {blendOption=Linear}
    boneToModify {boneName=root}
    scale=<0 0 0>
   }
   "Transform (Modify) Bone - Bone: root"(Translation_X=0, Translation_Y=Get ShieldPosOffset(), Translation_Z=0, Rotation="0,0,0", Scale="0,0,0") {
    rotationSpace=BCS_BoneSpace
    translationMode=BMM_Additive
    translationSpace=BCS_BoneSpace
    alphaBoolBlend {blendOption=Linear}
    boneToModify {boneName=root}
    scale=<0 0 0>
   }
   n1 = Blendspace Player 'TwinStretchCorn_Shield_Lag_BS'(X=0, Y=0) {blendSpace=/TwinStretchCorn/Assets/Animation/New_Animations/Shards_WEP/TwinStretchCorn_Shield_Lag_BS}
   AnimGraphNode_StateMachine_2 = Shield SM() {is=title AnimGraphNode_StateMachine}
   n2 = Apply Additive(
    Base=Apply Additive(
     Base=Apply Additive(Base=@AnimGraphNode_StateMachine_2, Additive=@n1) {alphaBoolBlend {blendOption=Linear}}
     Additive=Blendspace Player 'TwinStretchCorn_Shield_Lag_Spread_BS'(X=0, Y=0) {blendSpace=/TwinStretchCorn/Assets/Animation/New_Animations/Shards_WEP/TwinStretchCorn_Shield_Lag_Spread_BS}
    ) {alphaBoolBlend {blendOption=Linear}}
    Additive=Blendspace Player 'TwinStretchCorn_Shield_Lag_InOut_BS'!(X=0) -> (Pose) {blendSpace=/TwinStretchCorn/Assets/Animation/New_Animations/Shards_WEP/TwinStretchCorn_Shield_Lag_InOut_BS}
   ) {alphaBoolBlend {blendOption=Linear}}
   AnimGraphNode_StateMachine_1 = IdleNoise SM() {is=title AnimGraphNode_StateMachine}
   n3 = "Transform (Modify) Bone - Bone: root"!(
    ComponentPose=Local To Component(LocalPose=Apply Additive(Base=@n2, Additive=@AnimGraphNode_StateMachine_1) {alphaBoolBlend {blendOption=Linear}})
    Alpha
    Translation:Vector="0,0,0"
    Rotation="0.0,-90,0.0"
    Scale="0,0,0"
   ) -> (Pose) {
    rotationMode=BMM_Additive
    rotationSpace=BCS_BoneSpace
    translationSpace=BCS_BoneSpace
    alphaBoolBlend {blendOption=Linear}
    boneToModify {boneName=root}
    scale=<0 0 0>
   }
   n4 = "Transform (Modify) Bone - Bone: root"(ComponentPose=Local To Component(LocalPose=Local Space Ref Pose()), Translation="0,0,0", Rotation="0,0,0", Scale="0,0,0") {
    rotationSpace=BCS_BoneSpace
    scaleMode=BMM_Replace
    translationSpace=BCS_BoneSpace
    alphaBoolBlend {blendOption=Linear}
    boneToModify {boneName=root}
    scale=<0 0 0>
   }
   Output Pose(Result=Blend Poses by bool(
     BlendPose_0=Component To Local(ComponentPose=@n3)
     BlendPose_1=Component To Local(ComponentPose=@n4)
     BlendTime_0=0
     BlendTime_1=0
     bActiveValue=Get bIsBlocking() -> (bIsBlocking:bool)
    ) {blendType=Linear})
  ]}
 IdleNoise SM {
  entry=Wait
  path=AnimGraph.AnimGraphNode_StateMachine_1.IdleNoise SM
  ids=id
  nodes [
   AnimStateNode_0 = Wait() {is=title AnimStateNode}
   AnimStateNode_1 = Idle Noise() {is=title AnimStateNode}
   AnimStateTransitionNode_0 = AnimStateTransitionNode()
   AnimStateTransitionNode_1 = AnimStateTransitionNode() {bAutomaticRuleBasedOnSequencePlayerInState=true, crossfadeDuration=0.1}
  ]
  links [
   {
    dst=Wait
    rule=AnimGraph.AnimGraphNode_StateMachine_1.IdleNoise SM.AnimStateTransitionNode_1.Transition
    src=Idle Noise
    via=AnimStateTransitionNode_1
   }
   {
    dst=Idle Noise
    rule=AnimGraph.AnimGraphNode_StateMachine_1.IdleNoise SM.AnimStateTransitionNode_0.Transition
    src=Wait
    via=AnimStateTransitionNode_0
   }
  ]
 }
 Wait {
  path=AnimGraph.AnimGraphNode_StateMachine_1.IdleNoise SM.AnimStateNode_0.Wait
  nodes [
   n1 = "Sequence Player (additive) 'TwinStretchCorn_Shield_Center_Pose_WEP_Add'"() {bLoopAnimation=false, sequence=/TwinStretchCorn/Assets/Animation/New_Animations/Shards_WEP/TwinStretchCorn_Shield_Center_Pose_WEP_Add}
   Output Animation Pose(Result=@n1) {
    becomeRelevantFunction {bSelfContext=true, memberGuid=#1, memberName=OnInitialUpdate}
    node {becomeRelevantFunction {functionName=OnInitialUpdate}}
   }
  ]
 }
 Idle Noise {
  path=AnimGraph.AnimGraphNode_StateMachine_1.IdleNoise SM.AnimStateNode_1.Idle Noise
  nodes [Output Animation Pose(Result="Sequence Player (additive) 'TwinStretchCorn_Shield_Idle_Additive_WEP'"() {bLoopAnimation=false, sequence=&TwinStretchCorn_Shield_Idle_Additive_WEP})]
 }
 Transition {
  path=AnimGraph.AnimGraphNode_StateMachine_1.IdleNoise SM.AnimStateTransitionNode_0.Transition
  nodes [Result(bCanEnterTransition=Get bIsIdleNoise() -> (bIsIdleNoise:bool))]
 }
 Transition {
  path=AnimGraph.AnimGraphNode_StateMachine_1.IdleNoise SM.AnimStateTransitionNode_1.Transition
  nodes [Result() {
    becomeRelevantFunction {bSelfContext=true, memberGuid=#1, memberName=OnInitialUpdate}
    node {becomeRelevantFunction {functionName=OnInitialUpdate}}
   }]
 }
 Shield SM {
  entry=Equip
  path=AnimGraph.AnimGraphNode_StateMachine_2.Shield SM
  ids=id
  nodes [
   To Hit Reaction() {is=title AnimStateAliasNode, also {aliases [Hit Reaction, Idle]}}
   AnimStateNode_0 = Equip() {is=title AnimStateNode}
   AnimStateNode_1 = Idle()
   AnimStateNode_2 = Broken() {is=title AnimStateNode}
   AnimStateNode_3 = UnEquip() {is=title AnimStateNode}
   AnimStateNode_4 = Hit Reaction() {is=title AnimStateNode}
   AnimStateTransitionNode_0 = AnimStateTransitionNode() {bAutomaticRuleBasedOnSequencePlayerInState=true, crossfadeDuration=0.1}
   AnimStateTransitionNode_1 = AnimStateTransitionNode() {crossfadeDuration=0.1}
   AnimStateTransitionNode_2 = AnimStateTransitionNode()
   AnimStateTransitionNode_3 = AnimStateTransitionNode()
   AnimStateTransitionNode_4 = AnimStateTransitionNode()
   AnimStateTransitionNode_5 = AnimStateTransitionNode()
   AnimStateTransitionNode_7 = AnimStateTransitionNode() {bAutomaticRuleBasedOnSequencePlayerInState=true, crossfadeDuration=0}
   AnimStateTransitionNode_8 = AnimStateTransitionNode() {crossfadeDuration=0}
   AnimStateTransitionNode_9 = AnimStateTransitionNode() {crossfadeDuration=0.1}
  ]
  links [
   {
    dst=Equip
    rule=AnimGraph.AnimGraphNode_StateMachine_2.Shield SM.AnimStateTransitionNode_3.Transition
    src=Broken
    via=AnimStateTransitionNode_3
   }
   {
    dst=Idle
    rule=AnimGraph.AnimGraphNode_StateMachine_2.Shield SM.AnimStateTransitionNode_0.Transition
    src=Equip
    via=AnimStateTransitionNode_0
   }
   {
    dst=Idle
    rule=AnimGraph.AnimGraphNode_StateMachine_2.Shield SM.AnimStateTransitionNode_2.Transition
    src=Equip
    via=AnimStateTransitionNode_2
   }
   {
    dst=Idle
    rule=AnimGraph.AnimGraphNode_StateMachine_2.Shield SM.AnimStateTransitionNode_7.Transition
    src=Hit Reaction
    via=AnimStateTransitionNode_7
   }
   {
    dst=UnEquip
    rule=AnimGraph.AnimGraphNode_StateMachine_2.Shield SM.AnimStateTransitionNode_9.Transition
    src=Hit Reaction
    via=AnimStateTransitionNode_9
   }
   {
    dst=UnEquip
    rule=AnimGraph.AnimGraphNode_StateMachine_2.Shield SM.AnimStateTransitionNode_1.Transition
    src=Idle
    via=AnimStateTransitionNode_1
   }
   {
    dst=Hit Reaction
    rule=AnimGraph.AnimGraphNode_StateMachine_2.Shield SM.AnimStateTransitionNode_8.Transition
    src=To Hit Reaction
    via=AnimStateTransitionNode_8
   }
   {
    dst=Broken
    rule=AnimGraph.AnimGraphNode_StateMachine_2.Shield SM.AnimStateTransitionNode_5.Transition
    src=UnEquip
    via=AnimStateTransitionNode_5
   }
   {
    dst=Equip
    rule=AnimGraph.AnimGraphNode_StateMachine_2.Shield SM.AnimStateTransitionNode_4.Transition
    src=UnEquip
    via=AnimStateTransitionNode_4
   }
  ]
 }
 Equip {
  path=AnimGraph.AnimGraphNode_StateMachine_2.Shield SM.AnimStateNode_0.Equip
  nodes [
   n1 = Sequence Player 'TwinStretchCorn_Shield_Equip'() {bLoopAnimation=false, sequence=/TwinStretchCorn/Assets/Animation/New_Animations/Shards_WEP/TwinStretchCorn_Shield_Equip, startPosition=0.15}
   Output Animation Pose(Result=@n1)
  ]
 }
 Idle {
  path=AnimGraph.AnimGraphNode_StateMachine_2.Shield SM.AnimStateNode_1.Idle
  nodes [Output Animation Pose(Result=Sequence Player 'TwinStretchCorn_Shield_Idle'() {sequence=/TwinStretchCorn/Assets/Animation/New_Animations/Shards_WEP/TwinStretchCorn_Shield_Idle})]
 }
 Broken {
  path=AnimGraph.AnimGraphNode_StateMachine_2.Shield SM.AnimStateNode_2.Broken
  nodes [Output Animation Pose(Result=Sequence Player 'TwinStretchCorn_Shield_UnequipAlt'() {bLoopAnimation=false, sequence=/TwinStretchCorn/Assets/Animation/New_Animations/Shards_WEP/TwinStretchCorn_Shield_UnequipAlt})]
 }
 UnEquip {
  path=AnimGraph.AnimGraphNode_StateMachine_2.Shield SM.AnimStateNode_3.UnEquip
  nodes [Output Animation Pose(Result=Sequence Player 'TwinStretchCorn_Shield_Unequip'() {bLoopAnimation=false, sequence=/TwinStretchCorn/Assets/Animation/New_Animations/Shards_WEP/TwinStretchCorn_Shield_Unequip})]
 }
 Hit Reaction {
  path=AnimGraph.AnimGraphNode_StateMachine_2.Shield SM.AnimStateNode_4.Hit Reaction
  nodes [Output Animation Pose(Result=Sequence Player 'TwinStretchCorn_Shield_HitReaction_WEP'() {bLoopAnimation=false, sequence=/TwinStretchCorn/Assets/Animation/New_Animations/Shards_WEP/TwinStretchCorn_Shield_HitReaction_WEP})]
 }
 Transition {path=AnimGraph.AnimGraphNode_StateMachine_2.Shield SM.AnimStateTransitionNode_0.Transition, nodes [Result()]}
 Transition {
  path=AnimGraph.AnimGraphNode_StateMachine_2.Shield SM.AnimStateTransitionNode_1.Transition
  nodes [Result(bCanEnterTransition=NOT Boolean(A=Get bIsEquip() -> (bIsEquip:bool)))]
 }
 Transition {
  path=AnimGraph.AnimGraphNode_StateMachine_2.Shield SM.AnimStateTransitionNode_2.Transition
  nodes [Result(bCanEnterTransition=NOT Boolean(A=Get bIsEquip()))]
 }
 Transition {
  path=AnimGraph.AnimGraphNode_StateMachine_2.Shield SM.AnimStateTransitionNode_3.Transition
  nodes [Result(bCanEnterTransition=Get bIsEquip())]
 }
 Transition {
  path=AnimGraph.AnimGraphNode_StateMachine_2.Shield SM.AnimStateTransitionNode_4.Transition
  nodes [Result(bCanEnterTransition=Get bIsEquip())]
 }
 Transition {
  path=AnimGraph.AnimGraphNode_StateMachine_2.Shield SM.AnimStateTransitionNode_5.Transition
  nodes [Result(bCanEnterTransition=Get bIsShieldBroken() -> (bIsShieldBroken:bool))]
 }
 Transition {path=AnimGraph.AnimGraphNode_StateMachine_2.Shield SM.AnimStateTransitionNode_7.Transition, nodes [Result()]}
 Transition {
  path=AnimGraph.AnimGraphNode_StateMachine_2.Shield SM.AnimStateTransitionNode_8.Transition
  nodes [Result(bCanEnterTransition=Get bIsShieldHit() -> (bIsShieldHit:bool))]
 }
 Transition {
  path=AnimGraph.AnimGraphNode_StateMachine_2.Shield SM.AnimStateTransitionNode_9.Transition
  nodes [Result(bCanEnterTransition=NOT Boolean(A=Get bIsEquip()))]
 }
 EventGraph {links [], nodes []}
 ResetIsIdleNoise {nodes [
   ResetIsIdleNoise!() -> (then, Context:"Anim Update Context Structure (by ref)", Node:"Anim Node Reference Structure (by ref)") {is=title K2Node_FunctionEntry}
   > Set bIsIdleNoise(bIsIdleNoise:bool=false) -> (Output_Get:bool)
   > Return Node()
  ]}
]
variables []
