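; Timed Niagara anim notify for TwinStretchCorn that configures and updates its spawned effect during an animation.
; On begin, it sets rendering options and body-size and frontend variables, hiding the effect for frontend-hologram pawns or when the TwinStretchCorn weapon's owner is hovering.
; Each tick, it copies anim curve values into Niagara floats through configurable name pairs.
; Optionally, it also feeds the pawn's water state, depth, surface plane and water body into Niagara.
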(fn Received_NotifyBegin
  (params (MeshComp :type Skeletal%20Mesh%20Component%20Object%20Reference) (Animation :type Anim%20Sequence%20Base%20Object%20Reference) (TotalDuration :type Float%20%28single-precision%29) (EventReference :type Anim%20Notify%20Event%20Reference%20Structure%20%28by%20ref%29))
  (bind _returnvalue (Components|GetOwner MeshComp))
  (bind _print_debug (Variables|Debug|GetPrint%20Debug))
  (bind _returnvalue_2 (AnimNotify|GetSpawnedEffect MeshComp))
  (Utilities|FlowControl|Sequence
    (:then_0
      (bind _asfort_player_pawn (Utilities|Casting|CastToFortPlayerPawn _returnvalue))
      (Utilities|FlowControl|Sequence
        (:then_0
          (if (Variables|/Script/FortniteGame.FortPlayerPawn|GetbIsInFrontEndHologram _asfort_player_pawn)
            (Rendering|SetVisibility _returnvalue_2 false false)
            (else
              (Rendering|SetVisibility _returnvalue_2 true false))))
        (:then_1
          (Rendering|SetTranslucencySortDistanceOffset _returnvalue_2 (Variables|Default|GetTranslucency%20Sort%20Distance%20Offset))
          (bind _asniagara_particle_system_component (Utilities|Casting|CastToNiagaraComponent _returnvalue_2)
            (:then
              (bind _output_get (LocalVariables|SetNiagaraSystem _asniagara_particle_system_component))
              (Niagara|SetNiagaraVariable(Bool) _output_get "User.bIsLargeBodyPawn" (Variables|General|GetIsLargeBodyPawn))
              (Rendering|SetCastShadow _output_get (Variables|Default|GetCast%20Shadow))
              (bind _output_get_2 (LocalVariables|SetbIsFrontEnd (Reactive|Frontend|IsCosmeticObjectDisplayedinFrontendContext (Variables|GetMeshComp))))
              (if _output_get_2
                (Development|PrintString 0 "Setting Front End Variable" _print_debug _print_debug "(R=1.000000,G=0.000000,B=0.083156,A=1.000000)" 1e+01)
                (Niagara|SetNiagaraVariable(Bool) (LocalVariables|GetNiagaraSystem) "User.bIsFrontEndPreview" (LocalVariables|GetbIsFrontEnd))
                (else
                  (Development|PrintString 0 "Not setting Front End Variable" _print_debug _print_debug "(R=1.000000,G=0.000000,B=0.083156,A=1.000000)" 1e+01)))
              (if (Variables|Debug|Getis%20Glider%20Deploy%20Notify)
                (Utilities|Array|Add (Variables|/Script/Engine.ActorComponent|GetComponentTags (LocalVariables|GetNiagaraSystem)) "GliderDeploy"))
              (return true))
            (:CastFailed
              (return false))))))
    (:then_1
      (bind _asb_weap_twin_stretch_corn (Utilities|Casting|CastToB_Weap_TwinStretchCorn _returnvalue))
      (bind _asfort_player_pawn_2 (Utilities|Casting|CastToFortPlayerPawn (Actor|GetOwner _asb_weap_twin_stretch_corn)))
      (if (GameplayTags|HasMatchingGameplayTag _asfort_player_pawn_2 "(TagName=\"MovementMode.Hover\")")
        (Rendering|SetVisibility _returnvalue_2 false false)
        (else
          (Rendering|SetVisibility _returnvalue_2 true false))))))

(fn Received_NotifyTick
  (params (MeshComp :type Skeletal%20Mesh%20Component%20Object%20Reference) (Animation :type Anim%20Sequence%20Base%20Object%20Reference) (FrameDeltaTime :type Float%20%28single-precision%29) (EventReference :type Anim%20Notify%20Event%20Reference%20Structure%20%28by%20ref%29))
  (bind _player_pawn (LocalVariables|GetPlayer%20Pawn))
  (bind _returnvalue (WaterBody|GetCurrentWaterBody _player_pawn))
  (bind _player_pawn_2 (LocalVariables|GetPlayer%20Pawn))
  (bind _returnvalue_2 (WaterBody|GetCurrentWaterBody _player_pawn_2))
  (bind _bisinwater (LocalVariables|GetbIsInWater))
  (bind _waveparamtextureoffset (LocalVariables|GetwaveParamTextureOffset))
  (bind _niagarasystem (LocalVariables|GetNiagaraSystem))
  (bind _waterbodycomponent (Variables|/Script/Water.WaterBody|GetWaterBodyComponent _returnvalue))
  (Utilities|IsValid MeshComp
    (:Is%20Valid
      (bind _returnvalue_3 (AnimNotify|GetSpawnedEffect MeshComp))
      (Utilities|IsValid _returnvalue_3
        (:Is%20Valid
          (Variables|/Script/Engine.SceneComponent|SetbAbsoluteScale (Variables|Default|GetAbsolute%20Scale) _returnvalue_3)
          (bind _asniagara_particle_system_component (Utilities|Casting|CastToNiagaraComponent _returnvalue_3))
          (LocalVariables|SetNiagaraSystem _asniagara_particle_system_component)
          ;[0.3367 0.2681 1 0.185 40] Set niagara user float variables by curves on the animation instance.
          (for _array_element (Variables|Curves|GetAnimCurveParameterPair)
            (Niagara|SetNiagaraVariable(Float) (LocalVariables|GetNiagaraSystem) (Utilities|Struct|BreakAnimCurveParameterPair _array_element) (Animation|Curves|GetCurveValue (Components|SkeletalMesh|GetAnimInstance (Variables|GetMeshComp)) (.AnimCurveName_6_457883094DB38A3EF423919EDE4D33C1 (Utilities|Struct|BreakAnimCurveParameterPair _array_element)))))
          (if (Variables|Water|GetSet%20in%20Water%20Niagara%20Variable)
            (bind _asplayer_pawn_athena (Utilities|Casting|CastToPlayerPawn_Athena (Components|GetOwner (Variables|GetMeshComp)))
              (:then
                (bind _output_get (LocalVariables|SetPlayer%20Pawn _asplayer_pawn_athena))
                (bind _asfort_movement_comp_character (Utilities|Casting|CastToFortMovementComp_Character (Variables|/Script/Engine.Character|GetCharacterMovement _output_get))
                  (:then
                    (LocalVariables|SetFortMovementComp _asfort_movement_comp_character)
                    ;[1 1 1] qq
                    (Development|PrintString 0 (Utilities|String|Append "Is In Water: " (Utilities|String|ToString(Boolean) (or (Utilities|IsValid (WaterBody|GetCurrentWaterBody (LocalVariables|GetPlayer%20Pawn)) :!class /Script/Engine.KismetSystemLibrary) (WaterBody|IsSurfaceSwimming (LocalVariables|GetPlayer%20Pawn))))) (Variables|Debug|GetPrint%20Debug) (Variables|Debug|GetPrint%20Debug) "(R=0.000000,G=1.000000,B=0.377983,A=1.000000)")
                    (Utilities|FlowControl|Sequence
                      (:then_0
                        ;[1 1 1 1 40] Set local variables
                        (bind _output_get_2 (LocalVariables|SetbIsInWater (or (Utilities|IsValid (WaterBody|GetCurrentWaterBody (LocalVariables|GetPlayer%20Pawn)) :!class /Script/Engine.KismetSystemLibrary) (WaterBody|IsSurfaceSwimming (LocalVariables|GetPlayer%20Pawn)))))
                        (if _output_get_2
                          (LocalVariables|SetwaveParamTextureOffset (Variables|/Script/Water.WaterBody|GetWaterBodyIndex (WaterBody|GetCurrentWaterBody (LocalVariables|GetPlayer%20Pawn))))
                          (LocalVariables|SetwaterImmersionDepth (WaterBody|GetWaterImmersionDepth (LocalVariables|GetPlayer%20Pawn)))
                          (LocalVariables|SetwaterSurfaceLocation (Swimming|GetWaterSurfaceLocation (LocalVariables|GetFortMovementComp)))
                          (LocalVariables|SetwaterSurfaceNormal (Swimming|GetWaterSurfaceNormal (LocalVariables|GetFortMovementComp)))))
                      (:then_1
                        ;[1 1 1 1 40] Debug
                        (Development|PrintString 0 (Utilities|String|Append "In Water : " (Utilities|String|ToString(Boolean) (LocalVariables|GetbIsInWater))) (Variables|Debug|GetPrint%20Debug) (Variables|Debug|GetPrint%20Debug) "(R=0.000000,G=1.000000,B=0.419013,A=1.000000)")
                        (Development|PrintString 0 (Utilities|String|JoinStringArray (Utilities|Array|MakeArray (Utilities|String|Append "Depth : " (Utilities|String|ToString(Float) (LocalVariables|GetwaterImmersionDepth))) (Utilities|String|Append "Water Location : " (Utilities|String|ToString(Vector) (LocalVariables|GetwaterSurfaceLocation))) (Utilities|String|Append "Water Normal : " (Utilities|String|ToString(Vector) (LocalVariables|GetwaterSurfaceNormal))) (Utilities|String|Append "Wave Tex Offset : " (Utilities|String|ToString(Integer) (LocalVariables|GetwaveParamTextureOffset)))) " || ") (Variables|Debug|GetPrint%20Debug) (Variables|Debug|GetPrint%20Debug))
                        (Utilities|IsValid _returnvalue_2
                          (:Is%20Valid
                            (Development|PrintString 0 (Utilities|String|Append "Water Body Component: " (Utilities|GetDisplayName (Variables|/Script/Water.WaterBody|GetWaterBodyComponent _returnvalue_2))) (Variables|Debug|GetPrint%20Debug) (Variables|Debug|GetPrint%20Debug) "(R=0.000000,G=0.175929,B=0.300000,A=1.000000)"))))
                      (:then_2
                        (Niagara|SetNiagaraVariable(Bool) (LocalVariables|GetNiagaraSystem) "User.bIsInWater" _bisinwater)
                        ;[0 0.4523 1 0.138 80] Set Niagara Water User Variables
                        (if _bisinwater
                          (Niagara|SetNiagaraVariable(Int32) _niagarasystem "User.WaterBodyIndex" _waveparamtextureoffset)
                          (Niagara|SetNiagaraVariable(Float) _niagarasystem "User.WaterBodyIndex" (Math|Conversions|ToFloat(Integer) _waveparamtextureoffset))
                          (Niagara|SetNiagaraVariable(Float) (LocalVariables|GetNiagaraSystem) "User.WaterDepth" (LocalVariables|GetwaterImmersionDepth))
                          (Niagara|SetNiagaraVariable(Vector3) (LocalVariables|GetNiagaraSystem) "User.WaterPlaneLocation" (LocalVariables|GetwaterSurfaceLocation))
                          (Niagara|SetNiagaraVariable(Vector3) (LocalVariables|GetNiagaraSystem) "User.WaterPlaneNormal" (LocalVariables|GetwaterSurfaceNormal))
                          (Utilities|IsValid _waterbodycomponent
                            (:Is%20Valid
                              (Water|SetWaterBodyComponent (LocalVariables|GetNiagaraSystem) "WaterBodyActor" _waterbodycomponent)))))
                      (:then_3
                        (return true))))
                  (:CastFailed
                    (Development|PrintString 0 "Casting to Player Pawn Athena Failed!" (Variables|Debug|GetPrint%20Debug) (Variables|Debug|GetPrint%20Debug) "(R=1.000000,G=0.000000,B=0.083156,A=1.000000)" 1e+01)
                    (return false))))
              (:CastFailed
                (return false)))
            (else
              (return false))))))))
