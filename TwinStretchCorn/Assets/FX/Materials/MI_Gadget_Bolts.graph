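# Lit material for gadget bolt Niagara mesh particles, previewed on a combined bolts mesh.
# A sphere mask compares vertex color blue to the "WPO Away" dynamic parameter, yielding zero offset for matches and a huge negative offset otherwise, apparently to show one bolt variant per particle.
# Diffuse and normal textures supply the surface, and a mask texture blends toward a scalar parameter.
# A DebugTimeSine function node has no visible connections.
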Material /TwinStretchCorn/Assets/FX/Materials/MI_Gadget_Bolts
graph {
 properties {
  bUsedWithNiagaraMeshParticles=true
  pixelDepthOffsetMode=PDOM_Legacy
  previewMesh=/TwinStretchCorn/Assets/FX/Meshes/SM_Bolts_Combined
  usedShadingModels=Default Lit
 }
 nodes [
  n1 = VertexColor()
  n2 = DynamicParameter() {paramNames [WPO Away, Param2, Param3, Param4]}
  n3 = Constant() {r=0.03}
  LinearInterpolate(
   A=VectorParameter() {defaultValue=<rgba -8999999488 -8999999488 -8999999488 0>}.RGB
   Alpha=Ceil(None=SphereMask(A=@n1.B, B=@n2.WPO Away, Radius=@n3) {hardnessPercent=1})
  ) {constB=0}
  LinearInterpolate(
   B=ScalarParameter() {defaultValue=0.35}
   Alpha=TextureSample() {samplerType=SAMPLERTYPE_Masks, texture=/TwinStretchCorn/Assets/FX/Materials/Textures/T_Gadget_Bolts_S}.RGB
  )
  MaterialFunctionCall() {materialFunction=&DebugTimeSine}
  TextureSample() {texture=/TwinStretchCorn/Assets/FX/Materials/Textures/T_Gadget_Bolts_D}
  TextureSample() {samplerType=SAMPLERTYPE_Normal, texture=/TwinStretchCorn/Assets/FX/Materials/Textures/T_Gadget_Bolts_N}
 ]
}
