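# Unlit additive, two-sided material for magnetic line arc effects on Niagara mesh particles.
# SplineThicken widens the mesh with a taper at both ends, plus normal-direction wobble from a panned texture varied per particle.
# A SphereMask driven by dynamic parameters sweeps along the spline, adding to particle color for emissive adjustment.
# That mask also multiplies scrolling plasma, an edge fade across U and particle alpha.
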Material /TwinStretchCorn/Assets/FX/Materials/M_MagneticLines_Roundthumb
graph {
 properties {
  bUsedWithNiagaraMeshParticles=true
  blendMode=BLEND_Additive
  pixelDepthOffsetMode=PDOM_Legacy
  shadingModel=MSM_Unlit
  twoSided=true
  usedShadingModels=Unlit
 }
 nodes [
  n1 = ComponentMask(None=TextureCoordinate()) {g=true}
  n2 = OneMinus(None=@n1)
  n3 = Multiply(
   A=LinearInterpolate(B=ScalarParameter() {defaultValue=60, parameterName=Arc Thickness}, Alpha=Multiply(A=@n2, B=@n1))
   B=DynamicParameter() {paramNames [Sweep Along Spline, Spline Thickness, Radius, WPO Amout]}.Spline Thickness
  )
  n4 = VertexNormalWS()
  n5 = ParticleRandom()
  n6 = Multiply(
   A=LinearInterpolate(
    A=ScalarParameter() {defaultValue=0.35, parameterName=WPO Intensity Low}
    B=ScalarParameter() {defaultValue=0.75, parameterName=WPO Intensity High}
    Alpha=@n5
   )
   B=TextureCoordinate()
  )
  n7 = Multiply(A=Time(), B=LinearInterpolate(Alpha=@n5) {constA=0.25})
  n8 = TextureSample(UVs=Add(A=Panner(Coordinate=@n6, Time=@n7) {speedY=1}, B=LinearInterpolate(Alpha=@n5) {constA=0.25, constB=0.5})) {texture=&2ndInfectedMapLinear}
  Add(
   A=MaterialFunctionCall(WidthBase=@n3, WidthTip=@n3) {materialFunction=&SplineThicken}.WorldPositionOffset
   B=Multiply(A=Multiply(A=@n4, B=@n8.RGB), B=DynamicParameter() {paramNames [Sweep Along Spline, Spline Thickness, Radius, WPO Amout]}.WPO Amout)
  )
  n9 = DynamicParameter() {paramNames [Sweep Along Spline, Spline Thickness, Radius, WPO Amout]}
  n10 = ParticleRandom()
  n11 = SphereMask(
   A=ComponentMask(None=TextureCoordinate()) {g=true}
   B=Add(A=@n9.Sweep Along Spline, B=LinearInterpolate(Alpha=@n10) {constA=-0.12, constB=-0.25})
   Radius=Multiply(A=@n9.Radius, B=ScalarParameter() {defaultValue=0.15, parameterName=Arc Sweep Radius})
   Hardness=ScalarParameter() {defaultValue=0.35, parameterName=Arc Sweep hardness}
  )
  n12 = ParticleColor()
  MaterialFunctionCall(Emissive=Add(A=@n11, B=@n12.RGB)) {materialFunction=&MF_FX_EmissiveAdjustments}
  n13 = ComponentMask(None=TextureCoordinate()) {r=true}
  n14 = OneMinus(None=@n13)
  n15 = Multiply(A=TextureCoordinate(), B=ComponentMask(None=VectorParameter() {defaultValue=<rgba 1 0.5 0 0>}.RGB) {g=true, r=true})
  n16 = Multiply(
   A=TextureSample(UVs=Panner(Coordinate=@n15) {speedY=0.1}) {texture=/Game/Effects/Fort_Effects/Textures/Tile/T_Lightning_Plasma_Soft}.B
   B=ScalarParameter() {defaultValue=5, parameterName=Arc Boost}
  )
  Multiply(A=@n12.A, B=Multiply(A=Multiply(A=@n14, B=@n13), B=Multiply(A=@n11, B=@n16)))
  ParticleRelativeTime()
 ]
}
