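# Unlit, additive, two-sided Niagara sprite material that draws animated, noise-distorted concentric energy rings for the TwinStretchCorn effects.
# A panning noise texture, offset per particle by Particle Random, warps the center distance that drives both a radial falloff and outward-moving sharpened sine rings.
# Dynamic Parameters control ring distortion, period and exponent.
# The resulting Ring Mask scales particle color into MF_FX_EmissiveAdjustments for emissive and particle alpha into a depth-faded opacity.
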Material /TwinStretchCorn/Assets/FX/Materials/M_TwinStretchCorn_Energy
graph {
 properties {
  bUsedWithNiagaraSprites=true
  blendMode=BLEND_Additive
  pixelDepthOffsetMode=PDOM_Legacy
  shadingModel=MSM_Unlit
  twoSided=true
  usedShadingModels=Unlit
 }
 nodes [
  n1 = ParticleColor()
  n2 = NamedRerouteUsage()
  DepthFade(Opacity=Multiply(A=@n1.A, B=@n2), FadeDistance=ScalarParameter() {parameterName=Depth Fade Dist})
  MaterialFunctionCall(Emissive=Multiply(A=@n1.RGB, B=@n2)) {materialFunction=&MF_FX_EmissiveAdjustments}
  n3 = ParticleRandom()
  n4 = Add(A=Multiply(A=TextureCoordinate()) {constB=2}, B=@n3)
  n5 = Add(A=Time(), B=@n3)
  n6 = ConstantBiasScale(None=TextureSample(UVs=Panner(Coordinate=@n4, Time=@n5) {speedX=0.25, speedY=1}) {texture=/Game/Characters/Player/Male/Medium/Bodies/M_MED_Viper/Textures/Glitch/T_SoftNoise}.G) {bias=-0.5, scale=0.05}
  n7 = DynamicParameter() {defaultValue=<rgba 0.05 1 4 1>, paramNames [Ring Distortion, Ring Exponent, Ring Period, Param4]}
  n8 = Distance(A=TextureCoordinate(), B=Add(A=Multiply(A=@n6, B=@n7.Ring Distortion), B=Constant() {r=0.5}))
  n9 = Saturate(None=OneMinus(None=Multiply(A=@n8) {constB=2.2}))
  n10 = Add(A=Multiply(A=@n8, B=@n7.Ring Period) {constB=4}, B=Multiply(A=Time()) {constB=-1})
  n11 = Power(Base=Abs(None=Sine(None=@n10))) {constExponent=7}
  NamedRerouteDeclaration(None=Power(
    Base=Multiply(A=@n9, B=@n11)
    Exp=DynamicParameter() {defaultValue=<rgba 0.05 1 4 1>, paramNames [Ring Distortion, Ring Exponent, Ring Period, Param4]}.Ring Exponent
   ) {constExponent=1}) {name=Ring Mask, nodeColor=<rgba 0.5234 1 0 1>}
 ]
}
