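# Unlit, two-sided additive material for TwinStretchCorn's magnetic shield effect, usable on mesh particles and skeletal meshes.
# Flowmap-distorted scrolling sine rings, tuned by dynamic parameters, combine with panning caustics to boost emissive and also feed a normal-offset term.
# Color blends shield, timed damage-flash (custom primitive data) and ball-form tints, overridden by particle color at default settings.
# Opacity multiplies particle alpha, UV edge fades, inverted Fresnel, depth fade and panning plasma.
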Material /TwinStretchCorn/Assets/FX/Materials/M_TwinStretchCorn_MagneticShield
graph {
 properties {
  bCastRayTracedShadows=false
  bUsedWithMeshDeformer=true
  bUsedWithNiagaraMeshParticles=true
  bUsedWithSkeletalMesh=true
  blendMode=BLEND_Additive
  pixelDepthOffsetMode=PDOM_Legacy
  shadingModel=MSM_Unlit
  twoSided=true
  usedShadingModels=Unlit
 }
 nodes [
  n1 = TextureSample(UVs=Multiply(A=TextureCoordinate(), B=ScalarParameter() {defaultValue=2, parameterName=Ring Distortion Tiling})) {samplerType=SAMPLERTYPE_LinearColor, texture=/Game/Effects/Fort_Effects/Creatures/Elemental/Fire/MergedFireFlowMap}
  n2 = ScalarParameter() {defaultValue=1, parameterName=Ring Distortion Boost}
  n3 = DynamicParameter() {defaultValue=<rgba 0.12 0.25 1 1>, paramNames [Ring Distortion Intensity, Ring Size, Param3, Param4]}
  n4 = Add(A=TextureCoordinate(), B=Multiply(A=ComponentMask(None=@n1.RGB) {g=true, r=true}, B=Multiply(A=@n2, B=@n3.Ring Distortion Intensity)))
  n5 = ScalarParameter() {defaultValue=3, parameterName=Ring Amount}
  n6 = Add(
   A=ComponentMask(None=Multiply(A=@n4, B=@n5)) {g=true}
   B=Multiply(A=Time(), B=ScalarParameter() {defaultValue=1, parameterName=Ring Pan Speed})
  )
  n7 = ScalarParameter() {defaultValue=1, parameterName=Ring Size Boost}
  n8 = Multiply(
   A=SphereMask(A=Sine(None=@n6), B=Constant(), Radius=Multiply(A=@n7, B=@n3.Ring Size)) {hardnessPercent=0}
   B=ScalarParameter() {defaultValue=1, parameterName=Boost Rings}
  )
  Multiply(A=Multiply(A=VertexNormalWS(), B=@n8), B=ScalarParameter() {defaultValue=1, parameterName=Boost WPO}) {constB=10}
  n9 = ParticleColor()
  n10 = ScalarParameter() {defaultValue=1, parameterName=Use Particle Color}
  n11 = ComponentMask(None=LinearInterpolate(A=Constant() {r=1}, B=@n9.A, Alpha=@n10) {constAlpha=1}) {a=true}
  n12 = ComponentMask(None=TextureCoordinate()) {g=true}
  n13 = Multiply(
   A=Power(Base=OneMinus(None=@n12), Exp=ScalarParameter() {defaultValue=2, parameterName=Inside Edge Exponent})
   B=Power(Base=@n12, Exp=ScalarParameter() {defaultValue=2, parameterName=Outside Edge Exponent})
  )
  n14 = ComponentMask(None=TextureCoordinate()) {r=true}
  n15 = OneMinus(None=@n14)
  n16 = ScalarParameter() {defaultValue=2, parameterName=Inside Glow Exponent}
  n17 = ScalarParameter() {defaultValue=2, parameterName=Outside Glow Exponent}
  n18 = Multiply(A=@n13, B=Multiply(A=Power(Base=@n15, Exp=@n16), B=Power(Base=@n14, Exp=@n17)))
  n19 = ScalarParameter() {parameterName=Do U Fade}
  n20 = OneMinus(None=Fresnel() {exponent=1})
  n21 = ScalarParameter() {defaultValue=2, parameterName=FresnelExponent}
  n22 = DepthFade(Opacity=Multiply(A=LinearInterpolate(A=@n13, B=@n18, Alpha=@n19), B=Power(Base=@n20, Exp=@n21))) {fadeDistanceDefault=10}
  n23 = TextureCoordinate()
  n24 = ComponentMask(None=VectorParameter() {parameterName=Plasma UV Scale, defaultValue=<rgba 4 0.1 0 0>}.RGB) {g=true, r=true}
  n25 = Time()
  n26 = ScalarParameter() {defaultValue=0.12, parameterName=Plasma Pan Speed}
  n27 = TextureSample(UVs=Panner(Coordinate=Multiply(A=@n23, B=@n24), Time=Multiply(A=@n25, B=@n26)) {speedX=1}) {samplerType=SAMPLERTYPE_Masks, texture=/Game/Weapons/WeaponSkins/Wraps/Textures/PlasmaCharge/PlasmaCharge_M}
  n28 = ComponentMask(None=TextureCoordinate()) {g=true}
  n29 = OneMinus(None=SphereMask(
    A=Add(A=@n27.G, B=@n28)
    B=ScalarParameter() {defaultValue=0.01, parameterName=Plasma Start Position}
    Radius=ScalarParameter() {defaultValue=0.75, parameterName=Plasma Radius}
    Hardness=ScalarParameter() {defaultValue=0.5, parameterName=Plasma Hardness}
   ))
  n30 = ScalarParameter() {defaultValue=1, parameterName=Boost Plasma}
  Reroute(None=Multiply(A=Multiply(A=@n11, B=@n22) {desc=Particle Alpha}, B=Multiply(A=@n29, B=@n30)) {constA=1}) {desc=OPACITY}
  n31 = ScalarParameter() {bUseCustomPrimitiveData=true, parameterName=InputTime, primitiveDataIndex=1}
  n32 = Saturate(None=Sign(None=@n31))
  n33 = Divide(A=Subtract(A=Time(), B=@n31), B=Multiply(A=ScalarParameter() {defaultValue=0.2, parameterName=AnimDuration}) {constB=2})
  n34 = Sine(None=Multiply(A=Saturate(None=@n33)) {constB=0.5})
  n35 = LinearInterpolate(
   A=LinearInterpolate(
    A=VectorParameter() {parameterName=Shield Color, defaultValue=<rgba 0 0.3 25 80>}.RGB
    B=VectorParameter() {parameterName=Shield Damage Color, defaultValue=<rgba 3 0.1959 0 80>}.RGB
    Alpha=Multiply(A=@n32, B=@n34)
   ) {constAlpha=1}
   B=VectorParameter() {parameterName=Ball Form Color, defaultValue=<rgba 2.519 0.2 10 80>}.RGB
   Alpha=ScalarParameter() {parameterName=BallForm, primitiveDataIndex=1}
  ) {constAlpha=1}
  n36 = ComponentMask(None=LinearInterpolate(A=ComponentMask(None=@n35) {b=true, g=true, r=true}, B=@n9.RGB, Alpha=@n10) {constAlpha=1}) {b=true, g=true, r=true}
  n37 = Multiply(A=@n36, B=ScalarParameter() {defaultValue=3, parameterName=Ripple Boost}) {constB=4}
  n38 = Time()
  n39 = ScalarParameter() {defaultValue=0.1, parameterName=Caustics Pan Speed}
  n40 = TextureSample(UVs=Panner(Time=Multiply(A=@n38, B=@n39)) {speedX=1}) {samplerType=SAMPLERTYPE_Masks, texture=&PlasmaCharge_M}
  n41 = ScalarParameter() {defaultValue=1, parameterName=Boost Caustics}
  n42 = OneMinus(None=Add(A=Multiply(A=@n40.G, B=@n41), B=@n8) {constB=0})
  Reroute(None=MaterialFunctionCall(Emissive=LinearInterpolate(A=@n37, B=@n36, Alpha=@n42)) {materialFunction=&MF_FX_EmissiveAdjustments}.Result) {desc=EMISSIVE}
 ]
}
