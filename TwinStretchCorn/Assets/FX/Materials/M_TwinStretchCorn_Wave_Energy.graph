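# Translucent, two-sided unlit material for Niagara mesh particles, drawing a scrolling energy wave for the TwinStretchCorn effects.
# A panning noise warps the UVs of the noise, flame and gradient textures.
# Color comes from tinted noise layers times a vertical gradient and particle color, fed through MF_FX_EmissiveAdjustments.
# Alpha combines flame, gradient and noise, erodes along V by the Offset dynamic parameter, then applies an edge fade and particle alpha.
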Material /TwinStretchCorn/Assets/FX/Materials/M_TwinStretchCorn_Wave_Energy
graph {
 properties {
  bUsedWithNiagaraMeshParticles=true
  blendMode=BLEND_Translucent
  pixelDepthOffsetMode=PDOM_Legacy
  shadingModel=MSM_Unlit
  twoSided=true
  usedShadingModels=Unlit
 }
 nodes [
  n1 = DynamicParameter() {defaultValue=<rgba 0 1 1 1>, paramNames [Offset,Param2,Param3,Param4]}
  LinearInterpolate(Alpha=@n1.Offset) {constA=0.2, constB=0.4}
  MaterialFunctionCall() {materialFunction=&DebugTimeSine}
  n2 = TextureCoordinate()
  n3 = LinearInterpolate(A=Constant3Vector() {constant=<rgba 0.1543 0 2 1>}, Alpha=Power(Base=ComponentMask(None=@n2) {g=true}) {constExponent=1})
  n4 = Constant3Vector() {constant=<rgba 10 4.1097 2.9098 1>}
  n5 = Multiply(A=TextureCoordinate() {uTiling=7, vTiling=0.3})
  n6 = Multiply(A=TextureSample(UVs=Panner(Coordinate=@n5) {speedY=0.5}) {samplerType=SAMPLERTYPE_Masks, texture=/Game/Global/Textures/Default/Noise/T_WrapEdgeWear_Noise}.R) {constB=0.2}
  n7 = Panner(Coordinate=Add(A=TextureCoordinate() {uTiling=2}, B=@n6) {constB=0}) {speedY=0.5}
  n8 = TextureSample(UVs=@n7) {texture=/Game/Effects/Fort_Effects/Textures/Tile/T_Noise_02}
  n9 = Desaturation(None=@n8.RGB, Fraction=Constant())
  n10 = TextureCoordinate() {uTiling=8}
  n11 = TextureSample(UVs=Panner(Coordinate=Add(A=@n10, B=@n6) {constB=0}) {speedY=0.5}) {texture=/Game/Effects/Fort_Effects/Textures/Tile/T_Noise_03}
  n12 = Constant()
  n13 = Add(
   A=Multiply(A=Desaturation(None=@n11.RGB, Fraction=@n12), B=Constant3Vector() {constant=<rgba 28.1469 7.2885 50 1>})
   B=Add(A=Multiply(A=@n4, B=@n9), B=Constant3Vector() {constant=<rgba 0.6999 0.0337 3 1>})
  )
  n14 = ParticleColor()
  MaterialFunctionCall(Emissive=Multiply(A=Multiply(A=@n3, B=@n13), B=@n14.RGB)) {materialFunction=&MF_FX_EmissiveAdjustments}
  n15 = TextureSample(UVs=@n7) {samplerType=SAMPLERTYPE_LinearGrayscale, texture=/Game/Weapons/FORT_Melee/Pickaxe_Cube_Ninja_Male/FX/T_FluidFlameTile}
  n16 = Power(Base=Multiply(A=Power(Base=@n15.R)))
  n17 = TextureCoordinate() {uTiling=4, vTiling=-0.8}
  n18 = TextureSample(UVs=Add(A=Add(A=@n17, B=@n6) {constB=0}, B=Constant2Vector() {g=-0.1})) {texture=/Game/Weapons/FORT_Melee/Pickaxe_Cube_Ninja_Male/FX/T_Offset_Gradient_01}
  n19 = Add(A=ComponentMask(None=TextureCoordinate()) {g=true}, B=@n6)
  n20 = Subtract(A=OneMinus(None=Multiply(A=@n19)), B=@n1.Offset) {constB=0.2}
  n21 = Multiply(A=Multiply(A=Saturate(None=@n20))) {constB=20}
  n22 = Multiply(A=@n8.R, B=Multiply(A=LinearInterpolate(A=@n16, B=@n18.B), B=Saturate(None=@n21)))
  n23 = TextureCoordinate()
  n24 = Multiply(A=Power(Base=ComponentMask(None=@n23) {g=true}) {constExponent=3}) {constB=20}
  Multiply(A=@n14.A, B=Multiply(A=Multiply(A=@n22) {constB=10}, B=Saturate(None=@n24)))
 ]
}
