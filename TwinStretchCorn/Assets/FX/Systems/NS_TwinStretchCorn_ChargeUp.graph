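# Niagara charge-up effect for TwinStretchCorn, combining large pieces and purple-tinted bolt particles; all five emitters are currently disabled.
# Large_Pieces emitters burst per loop, easing from a system offset to User.FinalLocations entries, with User.Kill clearing the two mesh layers.
# Bolts orbit the NPC.FortniteNPC.PlayerPosition value, then blend into orbits around those locations.
# Bolts001 stays orbiting skeletal mesh samples (blend fixed at zero), and InterpolateOverTime smooths all motion.
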NiagaraSystem /TwinStretchCorn/Assets/FX/Systems/NS_TwinStretchCorn_ChargeUp
dependencies {
 usedDataInterfaces [
  &NiagaraDataInterfaceArrayFloat
  &NiagaraDataInterfaceMeshRendererInfo
  &NiagaraDataInterfaceCurve
  &NiagaraDataInterfaceColorCurve
 ]
 usedDynamicInputs [&RandomRangeInt]
 usedModules [
  &self:SystemSpawnScript.SystemScriptSource.SystemScriptGraph.NiagaraNodeAssignment_0.SetVariables_#1
  &SystemState
 ]
}
emitters {
 Bolts {
  bEnabled=false
  emitterUpdateScript [
   &EmitterState(
    Inactive Response=::NewEnumerator2
    Life Cycle Mode=::NewEnumerator1
    Loop Behavior=::NewEnumerator2
    Loop Count:NiagaraInt32=$User.LoopCount
    Loop Duration=0.5
    Scalability Mode=::NewEnumerator1
    UseLoopDelay=false
   ) {module=EmitterState001}
   &SpawnBurst_Instantaneous(
    Loop Count Limit=5
    Spawn Count=14
    Spawn Group=1
    Spawn Time=&SelectFloatFromArray(Direct Array Index=$Emitter.LoopCount, Float Selection Array=DataInterface() {propertyValues {FloatData [0,0.1,0.2,0.3,0.4]}})
    Use Spawn Probability=false
    UseLoopCountLimit=false
   )
  ]
  particleSpawnScript [
   &PartitionParticles(
    Number Of Partitions=/Niagara/DynamicInputs/Math/GetVectorArrayCount(Vector Array:NiagaraDataInterfaceArrayFloat3=$User.FinalLocations)
    Particles Per Partition=14
    Partition Mode:ENiagara_PartitionMode=::NewEnumerator1
   )
   &InitializeParticle(
    Color Mode=::NewEnumerator0
    Lifetime=5
    Mesh Renderer Array Visibility Mode=::NewEnumerator3
    Mesh Renderer Info=DataInterface() {propertyValues {MeshRenderer=&self:Bolts_2.NiagaraMeshRendererProperties_1}}
    Mesh Scale Mode=::NewEnumerator1
    Mesh Uniform Scale Max=1.25
    Mesh Uniform Scale Min=0.8
    UsePositionOffset=false
    Write Material Random=true
   )
   /SetVariables(Particles.Position=&TransformPosition(
     Position=&SelectVectorFromArray(Direct Array Index=$Particles.PartitionParticles.Partition, Vector Selection Array=$User.FinalLocations)
     Source Space=::Local
     Destination Space=::World
    )) {enabled=false, isSetParameters=true, module=SetVariables_#2}
   &DynamicMaterialParameters(
    Index 0 Param 1=&RandomRangeFloat(Minimum=0, Maximum=1)
    Module.Index0Param0Write=true
    Module.Index0Param1Write=true
    Module.Index0Param2Write=true
    Module.Index0Param3Write=true
   )
  ]
  particleUpdateScript [
   &ParticleState(Kill Particles When Lifetime Has Elapsed=false)
   &SkeletalMeshLocation(Bone Sampling Mode=::NewEnumerator0, MaskSpawnGroup=false, OverrideRandomSeed=false, Skeletal Mesh=$User.Skeletal Mesh)
   &UpdateMeshOrientation(
    Rotation Rate=&MaskFloatBySpawnGroup(
     Masked Float=&RandomRangeFloat(Minimum=0.25, Maximum=0.5)
     Passthrough Float=&RandomRangeFloat(Minimum=0.5, Maximum=1)
     Spawn Group Mask=1
    )
    Rotation Vector=&RandomRangeVector(Minimum=<-1 -1 -1>, Maximum=<1 1 1>)
   )
   &RotateAroundPoint(
    Radius:NiagaraFloat=&RandomRangeFloat(Minimum=40, Maximum=56)
    Rotation Center:NiagaraPosition=&AddVectorToPosition(
     Position=&TransformPosition(
      Position=&MakeVectorFromLinearColorRGB(Color=$NPC.FortniteNPC.PlayerPosition)
      Source Space=::World
      Destination Space=::Simulation
     )
     Vector=&RandomRangeVector(Minimum=<0 0 -30>, Maximum=<0 0 10>)
    )
    Rotation Phase:NiagaraFloat=&RandomRangeFloat(Minimum=0, Maximum=1)
    Rotation Rate:NiagaraFloat=&RandomRangeFloat(Minimum=0.03, Maximum=0.15)
   )
   &RotateAroundPoint(
    Derive the Rotation Axis=false
    Radius:NiagaraFloat=&RandomRangeFloat(Minimum=15, Maximum=55)
    Rotation Axis:Vector3f=&RandomRangeVector(Minimum=<0 0 0>, Maximum=<1 1 1>)
    Rotation Center:NiagaraPosition=&SelectVectorFromArray(Direct Array Index=$Particles.PartitionParticles.Partition, Direct Array Mode=::NewEnumerator1, Vector Selection Array=$User.FinalLocations)
    Rotation Phase:NiagaraFloat=&RandomRangeFloat(Minimum=0, Maximum=1)
    Rotation Rate:NiagaraFloat=&RandomRangeFloat(Minimum=0.05, Maximum=0.15)
   ) {module=RotateAroundPoint001}
   /SetVariables(Particles.Position=&Lerp_Vector(
     A=$Output.RotateAroundPoint.OutputPosition
     B=$Output.RotateAroundPoint001.OutputPosition
     Alpha=&FloatFromCurve(
      FloatCurve=DataInterface() {propertyValues {Curve {keys [
          {interpMode=RCIM_Cubic, leaveTangent=4.2253, tangentMode=RCTM_Break, time=0.5, value=0.05}
          {interpMode=RCIM_Cubic, leaveTangent=0.917, tangentMode=RCTM_Break, time=1, value=1}
         ]}}}
      CurveIndex=$Particles.Age
     )
    )) {isSetParameters=true, module=SetVariables_#3}
   &InterpolateOverTime(
    Rate Of Change:NiagaraFloat=&RandomRangeFloat(Minimum=1, Maximum=20)
    Target Vector Value:NiagaraPosition=$Particles.Position
    Variable Type=::NewEnumerator2
   )
   /SetVariables(Particles.Position=$Particles.InterpolateOverTime.InterpolateOverTime_V3_.Moving Average) {isSetParameters=true, module=SetVariables_#4}
   &FN_Sprite_Attributes(
    Camera Offset=-50
    Color=<rgba 0.4621 0 1 0.01>
    Create Camera Offset Attribute=true
    Create Color Attribute=true
    Create SpriteSize Attribute=true
    SpriteSize=<50 50>
    Write to Module Namespace=true
   )
   /SetVariables(Particles.Sprite Dynamic:Vector4f=&MakeVector4(X=2, Y=&RandomRangeFloat(Minimum=50, Maximum=75), Z=0, W=0)) {isSetParameters=true, module=SetVariables_#5}
  ]
  rendererClasses [&NiagaraSpriteRendererProperties, &NiagaraMeshRendererProperties]
  renderers [
   {rendererClass=&NiagaraSpriteRendererProperties}
   {rendererClass=&NiagaraMeshRendererProperties, rendererIndex=1}
  ]
 }
 Bolts001 {
  bEnabled=false
  emitterUpdateScript [
   &EmitterState(
    Inactive Response=::NewEnumerator2
    Life Cycle Mode=::NewEnumerator1
    Loop Behavior=::NewEnumerator1
    Loop Duration=0.5
    Scalability Mode=::NewEnumerator1
    UseLoopDelay=false
   ) {module=EmitterState001}
   &SpawnBurst_Instantaneous(Loop Count Limit=5, Spawn Count=14, Spawn Group=1, Use Spawn Probability=false, UseLoopCountLimit=false)
  ]
  particleSpawnScript [
   &PartitionParticles(
    Number Of Partitions=&GetVectorArrayCount(Vector Array:NiagaraDataInterfaceArrayFloat3=$User.FinalLocations)
    Particles Per Partition=14
    Partition Mode:ENiagara_PartitionMode=::NewEnumerator1
   )
   &InitializeParticle(
    Color Mode=::NewEnumerator0
    Lifetime=5
    Mesh Renderer Array Visibility Mode=::NewEnumerator3
    Mesh Renderer Info=DataInterface() {propertyValues {MeshRenderer=&self:Bolts001_1.NiagaraMeshRendererProperties_0}}
    Mesh Scale Mode=::NewEnumerator1
    Mesh Uniform Scale Max=0.5
    Mesh Uniform Scale Min=0.35
    UsePositionOffset=false
    Write Material Random=true
   )
   /SetVariables(Particles.Position=&TransformPosition(
     Position=&SelectVectorFromArray(Direct Array Index=$Particles.PartitionParticles.Partition, Vector Selection Array=$User.FinalLocations)
     Source Space=::Local
     Destination Space=::World
    )) {enabled=false, isSetParameters=true, module=SetVariables_#2}
   &InitialMeshOrientation(Use Rotation Vector=false)
   &DynamicMaterialParameters(
    Index 0 Param 1=&RandomRangeFloat(Minimum=0, Maximum=1)
    Module.Index0Param0Write=true
    Module.Index0Param1Write=true
    Module.Index0Param2Write=true
    Module.Index0Param3Write=true
   )
  ]
  particleUpdateScript [
   &ParticleState(Kill Particles When Lifetime Has Elapsed=false)
   &SkeletalMeshLocation(Bone Sampling Mode=::NewEnumerator0, MaskSpawnGroup=false, OverrideRandomSeed=false, Skeletal Mesh=$User.Skeletal Mesh)
   &UpdateMeshOrientation(
    Rotation Rate=&MaskFloatBySpawnGroup(
     Masked Float=&RandomRangeFloat(Minimum=0.12, Maximum=0.25)
     Passthrough Float=&RandomRangeFloat(Minimum=0.5, Maximum=1)
     Spawn Group Mask=1
    )
    Rotation Vector=&RandomRangeVector(Minimum=<-1 -1 -1>, Maximum=<1 1 1>)
   )
   &RotateAroundPoint(
    Radius:NiagaraFloat=&RandomRangeFloat(Minimum=7, Maximum=12)
    Rotation Center:NiagaraPosition=&AddVectorToPosition(
     Position=$Particles.SkeletalMeshLocation.SampledPosition
     Vector=&RandomRangeVector(Minimum=<0 0 -10>, Maximum=<0 0 10>)
    )
    Rotation Phase:NiagaraFloat=&RandomRangeFloat(Minimum=0, Maximum=1)
    Rotation Rate:NiagaraFloat=&RandomRangeFloat(Minimum=0.03, Maximum=0.15)
   )
   &RotateAroundPoint(
    Derive the Rotation Axis=false
    Radius:NiagaraFloat=&RandomRangeFloat(Minimum=30, Maximum=75)
    Rotation Axis:Vector3f=&RandomRangeVector(Minimum=<0 0 0>, Maximum=<1 1 1>)
    Rotation Center:NiagaraPosition=&SelectVectorFromArray(Direct Array Index=$Particles.PartitionParticles.Partition, Direct Array Mode=::NewEnumerator1, Vector Selection Array=$User.FinalLocations)
    Rotation Phase:NiagaraFloat=&RandomRangeFloat(Minimum=0, Maximum=1)
    Rotation Rate:NiagaraFloat=&RandomRangeFloat(Minimum=0.05, Maximum=0.15)
   ) {module=RotateAroundPoint001}
   /SetVariables(Particles.Position=&Lerp_Vector(A=$Output.RotateAroundPoint.OutputPosition, B=$Output.RotateAroundPoint001.OutputPosition, Alpha=0)) {isSetParameters=true, module=SetVariables_#3}
   &InterpolateOverTime(
    Rate Of Change:NiagaraFloat=&RandomRangeFloat(Minimum=1, Maximum=20)
    Target Vector Value:NiagaraPosition=$Particles.Position
    Variable Type=::NewEnumerator2
   )
   /SetVariables(Particles.Position=$Particles.InterpolateOverTime.InterpolateOverTime_V3_.Moving Average) {isSetParameters=true, module=SetVariables_#4}
   &FN_Sprite_Attributes(
    Camera Offset=-50
    Color=<rgba 0.4621 0 1 0.1>
    Create Camera Offset Attribute=true
    Create Color Attribute=true
    Create SpriteSize Attribute=true
    SpriteSize=<50 50>
    Write to Module Namespace=true
   )
   /SetVariables(Particles.Sprite Dynamic:Vector4f=&MakeVector4(X=2, Y=&RandomRangeFloat(Minimum=50, Maximum=75), Z=0, W=0)) {isSetParameters=true, module=SetVariables_#5}
  ]
  rendererClasses [&NiagaraMeshRendererProperties, &NiagaraSpriteRendererProperties]
  renderers [
   {rendererClass=&NiagaraMeshRendererProperties, rendererIndex=0}
   {rendererClass=&NiagaraSpriteRendererProperties, rendererIndex=1}
  ]
 }
 Large_Pieces {
  bEnabled=false
  emitterUpdateScript [
   &EmitterState(
    Inactive Response=::NewEnumerator2
    Life Cycle Mode=::NewEnumerator1
    Loop Behavior=::NewEnumerator2
    Loop Count:NiagaraInt32=$User.LoopCount
    Loop Duration=0.5
    Scalability Mode=::NewEnumerator1
    UseLoopDelay=false
   ) {module=EmitterState001}
   &SpawnBurst_Instantaneous(
    Loop Count Limit=5
    Spawn Time=&SelectFloatFromArray(Direct Array Index=$Emitter.LoopCount, Float Selection Array=DataInterface() {propertyValues {FloatData [0,0.1,0.2,0.3,0.4]}})
    Use Spawn Probability=false
    UseLoopCountLimit=false
   )
  ]
  particleSpawnScript [
   &PartitionParticles(Number Of Partitions=5)
   &InitializeParticle(
    Color Mode=::NewEnumerator0
    Lifetime=10
    Mesh Renderer Array Visibility Mode=::NewEnumerator3
    Mesh Renderer Info=DataInterface() {propertyValues {MeshRenderer=&self:Large_Pieces_0.NiagaraMeshRendererProperties_0}}
    Mesh Scale Mode=::NewEnumerator1
    Mesh Uniform Scale Max=2.5
    Mesh Uniform Scale Min=2
    Sprite Size Mode=::NewEnumerator3
    Uniform Sprite Size=150
    UsePositionOffset=false
    Write Material Random=true
   )
   /SetVariables(Particles.Position=&SelectVectorFromArray(Direct Array Index=$Emitter.LoopCount, Direct Array Mode=::NewEnumerator1, Vector Selection Array=$User.FinalLocations)) {isSetParameters=true, module=SetVariables_#6}
   &InitialMeshOrientation(Use Rotation Vector=false)
  ]
  particleUpdateScript [
   &ParticleState(Kill Particles When Lifetime Has Elapsed=false)
   &UpdateMeshOrientation(
    Rotation Rate=&RandomRangeFloat(Minimum=0.1, Maximum=0.15)
    Rotation Vector=&RandomRangeVector(Minimum=<0 0 0>, Maximum=<1 1 1>)
   )
   &Color()
   /SetVariables(Particles.Position=&Lerp_Vector(
     A=$System.offset
     B=Linked() {linkedVariable {name=Particles.Initial.Position, type=NiagaraPosition}}
     Alpha=&FloatFromCurve(
      FloatCurve=DataInterface() {propertyValues {Curve {keys [
          {interpMode=RCIM_Cubic, leaveTangent=6.4881, tangentMode=RCTM_Break, time=0, value=0}
          {interpMode=RCIM_Cubic, leaveTangent=0.917, tangentMode=RCTM_Break, time=0.25, value=1}
         ]}}}
      CurveIndex=$Particles.Age
     )
    )) {isSetParameters=true, module=SetVariables_#7}
   &InterpolateOverTime(
    Rate Of Change:NiagaraFloat=&RandomRangeFloat(Minimum=10, Maximum=15)
    Target Vector Value:NiagaraPosition=$Particles.Position
    Variable Type=::NewEnumerator2
   )
   /SetVariables(Particles.Position=$Particles.InterpolateOverTime.InterpolateOverTime_V3_.Moving Average) {isSetParameters=true, module=SetVariables_#8}
   &KillParticles(Kill Particles=$User.Kill)
  ]
  rendererClasses [&NiagaraMeshRendererProperties]
  renderers [{rendererClass=&NiagaraMeshRendererProperties, rendererIndex=0}]
 }
 Large_Pieces001 {
  bEnabled=false
  emitterUpdateScript [
   &EmitterState(
    Inactive Response=::NewEnumerator2
    Life Cycle Mode=::NewEnumerator1
    Loop Behavior=::NewEnumerator2
    Loop Count:NiagaraInt32=$User.LoopCount
    Loop Duration=0.5
    Scalability Mode=::NewEnumerator1
    UseLoopDelay=false
   ) {module=EmitterState001}
   &SpawnBurst_Instantaneous(
    Loop Count Limit=5
    Spawn Time=&SelectFloatFromArray(Direct Array Index=$Emitter.LoopCount, Float Selection Array=DataInterface() {propertyValues {FloatData [0,0.1,0.2,0.3,0.4]}})
    Use Spawn Probability=false
    UseLoopCountLimit=false
   )
  ]
  particleSpawnScript [
   &InitializeParticle(
    Color=&RandomRangeLinearColor(Minimum=<rgba 0.3172 0 1 0.03>, Maximum=<rgba 0.6015 0 1 0.1>)
    Lifetime=10
    Mesh Scale Mode=::NewEnumerator1
    Mesh Uniform Scale Max=0.5
    Mesh Uniform Scale Min=0.25
    Sprite Rotation Mode=::NewEnumerator2
    Sprite Size Mode=::NewEnumerator3
    Sprite UV Mode=::NewEnumerator3
    Uniform Sprite Size=150
    UsePositionOffset=false
    Write Material Random=true
   )
   /SetVariables(Particles.Position=&SelectVectorFromArray(Direct Array Index=$Emitter.LoopCount, Direct Array Mode=::NewEnumerator1, Vector Selection Array=$User.FinalLocations)) {isSetParameters=true, module=SetVariables_#6}
   &InitialMeshOrientation(Use Rotation Vector=false)
  ]
  particleUpdateScript [
   &ParticleState(Kill Particles When Lifetime Has Elapsed=false)
   &UpdateMeshOrientation(
    Rotation Rate=&RandomRangeFloat(Minimum=0.05, Maximum=0.1)
    Rotation Vector=&RandomRangeVector(Minimum=<0 0 0>, Maximum=<1 1 1>)
   )
   /SetVariables(Particles.Position=&Lerp_Vector(
     A=$System.offset
     B=Linked() {linkedVariable {name=Particles.Initial.Position, type=NiagaraPosition}}
     Alpha=&FloatFromCurve(
      FloatCurve=DataInterface() {propertyValues {Curve {keys [
          {interpMode=RCIM_Cubic, leaveTangent=6.4881, tangentMode=RCTM_Break, time=0, value=0}
          {interpMode=RCIM_Cubic, leaveTangent=0.917, tangentMode=RCTM_Break, time=0.25, value=1}
         ]}}}
      CurveIndex=$Particles.Age
     )
    )) {isSetParameters=true, module=SetVariables_#7}
   &InterpolateOverTime(
    Rate Of Change:NiagaraFloat=&RandomRangeFloat(Minimum=10, Maximum=15)
    Target Vector Value:NiagaraPosition=$Particles.Position
    Variable Type=::NewEnumerator2
   )
   /SetVariables(Particles.Position=$Particles.InterpolateOverTime.InterpolateOverTime_V3_.Moving Average) {isSetParameters=true, module=SetVariables_#8}
   &DynamicMaterialParameters(
    Index 0 Param 1=&ModuloFloat(A=$Particles.Age, B=1)
    Module.Index0Param0Write=true
    Module.Index0Param1Write=true
    Module.Index0Param2Write=true
    Module.Index0Param3Write=true
   )
   /SetVariables(Particles.MeshColor:LinearColor=&MultiplyLinearColor(
     Linear Color=$Particles.Color
     Scale Factor=&MakeVector4_FromLinearColor(LinearColor=&ColorFromCurve(
       ColorCurve=DataInterface() {propertyValues {
         AlphaCurve {keys [<at 0 0>,<at 0.4447 1 RCIM_Cubic>,<at 1 0 RCIM_Cubic>]}
         BlueCurve {keys [<at 0 1 RCIM_Cubic>,<at 1 1 RCIM_Cubic>]}
         GreenCurve {keys [<at 0 1 RCIM_Cubic>,<at 1 1 RCIM_Cubic>]}
         RedCurve {keys [<at 0 1 RCIM_Cubic>,<at 1 1 RCIM_Cubic>]}
        }}
       CurveIndex=&ModuloFloat(A=$Particles.Age, B=&RandomRangeFloat(Minimum=1, Maximum=2))
      ))
    )) {isSetParameters=true}
   &FN_Sprite_Attributes(
    Camera Offset=-50
    Create Camera Offset Attribute=true
    Create SpriteSize Attribute=true
    SpriteSize=<150 150>
    Write to Module Namespace=true
   )
   /SetVariables(Particles.Sprite Dynamic:Vector4f=&MakeVector4(X=2, Y=&RandomRangeFloat(Minimum=50, Maximum=75), Z=0, W=0)) {isSetParameters=true}
  ]
  rendererClasses [&NiagaraSpriteRendererProperties]
  renderers [
   {rendererClass=&NiagaraSpriteRendererProperties}
   {rendererClass=&NiagaraSpriteRendererProperties, rendererIndex=1}
  ]
 }
 Large_Pieces002 {
  bEnabled=false
  emitterUpdateScript [
   &EmitterState(
    Inactive Response=::NewEnumerator2
    Life Cycle Mode=::NewEnumerator1
    Loop Behavior=::NewEnumerator2
    Loop Count:NiagaraInt32=$User.LoopCount
    Loop Duration=0.5
    Scalability Mode=::NewEnumerator1
    UseLoopDelay=false
   ) {module=EmitterState001}
   &SpawnBurst_Instantaneous(
    Loop Count Limit=5
    Spawn Count=7
    Spawn Time=&SelectFloatFromArray(Direct Array Index=$Emitter.LoopCount, Float Selection Array=DataInterface() {propertyValues {FloatData [0,0.1,0.2,0.3,0.4]}})
    Use Spawn Probability=false
    UseLoopCountLimit=false
   )
  ]
  particleSpawnScript [
   &PartitionParticles(Number Of Partitions=5)
   &InitializeParticle(
    Color=<rgba 25.4758 0 50 1>
    Lifetime=10
    Mesh Scale Mode=::NewEnumerator1
    Mesh Uniform Scale Max=5
    Mesh Uniform Scale Min=1.5
    Sprite Size Mode=::NewEnumerator3
    Uniform Sprite Size=150
    UsePositionOffset=false
    Write Material Random=true
   )
   /SetVariables(Particles.Position=&SelectVectorFromArray(Direct Array Index=$Emitter.LoopCount, Direct Array Mode=::NewEnumerator1, Vector Selection Array=$User.FinalLocations)) {isSetParameters=true, module=SetVariables_#6}
   &InitialMeshOrientation(Use Rotation Vector=false)
  ]
  particleUpdateScript [
   &ParticleState(Kill Particles When Lifetime Has Elapsed=false)
   &ScaleMeshSize(Scale Factor=&VectorFromFloat(Value=&Lerp_Float(
      A=0.5
      B=1
      Alpha=&ModuloFloat(A=$Particles.Age, B=&RandomRangeFloat(Minimum=0.25, Maximum=1))
     )))
   &UpdateMeshOrientation(Rotation Rate=0.05)
   /SetVariables(Particles.Position=&Lerp_Vector(
     A=$System.offset
     B=Linked() {linkedVariable {name=Particles.Initial.Position, type=NiagaraPosition}}
     Alpha=&FloatFromCurve(
      FloatCurve=DataInterface() {propertyValues {Curve {keys [
          {interpMode=RCIM_Cubic, leaveTangent=6.4881, tangentMode=RCTM_Break, time=0, value=0}
          {interpMode=RCIM_Cubic, leaveTangent=0.917, tangentMode=RCTM_Break, time=0.25, value=1}
         ]}}}
      CurveIndex=$Particles.Age
     )
    )) {isSetParameters=true, module=SetVariables_#7}
   &InterpolateOverTime(
    Rate Of Change:NiagaraFloat=&RandomRangeFloat(Minimum=10, Maximum=15)
    Target Vector Value:NiagaraPosition=$Particles.Position
    Variable Type=::NewEnumerator2
   )
   /SetVariables(Particles.Position=$Particles.InterpolateOverTime.InterpolateOverTime_V3_.Moving Average) {isSetParameters=true, module=SetVariables_#8}
   &KillParticles(Kill Particles=$User.Kill)
   &DynamicMaterialParameters(
    Index 0 Param 1=&Lerp_Float(
     A=0
     B=1
     Alpha=&ModuloFloat(A=$Particles.Age, B=&RandomRangeFloat(Minimum=1, Maximum=2))
    )
    Index 0 Param 2=&RandomRangeFloat(Minimum=0.35, Maximum=0.75)
    Index 0 Param 3=&RandomRangeFloat(Minimum=0.5, Maximum=2)
    Index 0 Param 4=&RandomRangeFloat(Minimum=3, Maximum=10)
    Module.Index0Param0Write=true
    Module.Index0Param1Write=true
    Module.Index0Param2Write=true
    Module.Index0Param3Write=true
   )
   &ScaleColor(
    Scale Alpha=&FloatFromCurve(
     FloatCurve=DataInterface() {propertyValues {Curve {keys [<at 0 0>,<at 0.5 1>,<at 1 0>]}}}
     CurveIndex=&ModuloFloat(A=$Particles.Age, B=&RandomRangeFloat(Minimum=0.25, Maximum=3))
    )
    ScaleA=true
    ScaleRGB=true
   )
  ]
  rendererClasses [&NiagaraMeshRendererProperties]
  renderers [{rendererClass=&NiagaraMeshRendererProperties, rendererIndex=0}]
 }
}
summary {
 systemName=NS_TwinStretchCorn_ChargeUp
 emitters [
  {bEnabled=false, emitterName=Large_Pieces, rendererClasses [&NiagaraMeshRendererProperties]}
  {
   bEnabled=false
   emitterName=Bolts
   rendererClasses [&NiagaraSpriteRendererProperties, &NiagaraMeshRendererProperties]
  }
  {bEnabled=false, emitterName=Large_Pieces001, rendererClasses [&NiagaraSpriteRendererProperties]}
  {
   bEnabled=false
   emitterName=Bolts001
   rendererClasses [&NiagaraMeshRendererProperties, &NiagaraSpriteRendererProperties]
  }
  {bEnabled=false, emitterName=Large_Pieces002, rendererClasses [&NiagaraMeshRendererProperties]}
 ]
 userVariables [
  User.FinalLocations {
   type=NiagaraDataInterfaceArrayFloat3
   description=""
   defaultValue {
    struct=NiagaraExt_VariableValue_DataInterface
    value {
     dataInterface=&self:NiagaraDataInterfaceArrayFloat3_0
     dataInterfaceClass=&NiagaraDataInterfaceArrayFloat3
    }
   }
  }
  User.Kill {type=NiagaraBool}
  User.LoopCount {type=NiagaraInt32, defaultValue=5}
  User.Skeletal Mesh {
   type=NiagaraDataInterfaceSkeletalMesh
   defaultValue {
    struct=NiagaraExt_VariableValue_DataInterface
    value {dataInterface=&self:NiagaraDataInterfaceSkeletalMesh_0}
   }
  }
 ]
}
