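# Looping foley sound cue for the TwinStretchCorn weapon while equipped, using weapon reload attenuation and the weapon foley sound class.
# A looping wave player feeds an Enveloper, whose volume curve eases the loop down from full level over its first few seconds.
# The cue plays quietly at a slightly raised pitch and overrides concurrency to cap simultaneous instances.
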SoundCue /TwinStretchCorn/Assets/Audio/Cues/SC_TwinStretchCorn_Equipped_Loop
root=Enveloper_0
nodes [
 {
  id=Enveloper_0
  children [WavePlayer_4]
  properties {pitchCurve {editorCurveData {keys [{time=0}]}}, volumeCurve {editorCurveData {keys [<at 0 1>,<at 5 0.7>]}}}
 }
 {id=WavePlayer_4, properties {bLooping=true, soundWaveAssetPtr=/TwinStretchCorn/Assets/Audio/Waves/TwinStretchCorn_Equipped_Loop}}
]
properties {
 attenuationSettings=&Fort_Weapon_Reload_Att
 bOverrideConcurrency=true
 maxDistance=2250
 pitchMultiplier=1.1
 soundClassObject=&Fort_WeaponFoley
 volumeMultiplier=0.3
 concurrencyOverrides {maxCount=3}
}
