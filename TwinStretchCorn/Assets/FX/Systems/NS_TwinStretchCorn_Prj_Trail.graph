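# Niagara projectile trail for TwinStretchCorn, pairing a sprite wake with a mesh pressure front.
# BulletTimeSprite spawns short-lived, growing, fading sprites, using normalized owner velocity length to gate spawn rate to near zero when stationary.
# PressureFront bursts a persistent mesh that faces along owner velocity until the emitter deactivates.
# A system Timeline's normalized time ramps both emitters' color and alpha and grows the front, while dynamic material parameters feed their materials.
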NiagaraSystem /TwinStretchCorn/Assets/FX/Systems/NS_TwinStretchCorn_Prj_Trail
dependencies {
 usedDataInterfaces [&NiagaraDataInterfaceCurve, &NiagaraDataInterfaceVectorCurve]
 usedDynamicInputs [&SimulationPosition, /Game/Effects/Niagara/DynamicInputs/State/FN_IsExecutionStateActive]
 usedModules [
  &self:SystemSpawnScript.SystemScriptSource.SystemScriptGraph.NiagaraNodeAssignment_0.SetVariables_#1
  &SystemState
  &Timeline
 ]
}
emitters {
 BulletTimeSprite {
  emitterUpdateScript [
   &EmitterState()
   &SpawnRate(
    SpawnRate=&Lerp_Float(
     A=0
     B=15
     Alpha=&VectorLength(Vector=&NormalizeVector(Vector To Normalize=$Engine.Owner.Velocity, Safe Normalize=true, Safe Fallback Threshold=0, Safe Fallback Vector=<0 0 0.001>))
    )
    Use Spawn Probability=false
   )
  ]
  particleSpawnScript [&InitializeParticle(
    Color=<rgba 0.1667 0 1 1>
    Lifetime=0.66
    Mass Mode=::NewEnumerator1
    Position Offset=<0 0 100>
    Ribbon Width=300
    Ribbon Width Mode=::NewEnumerator1
    Sprite Rotation Mode=::NewEnumerator2
    Sprite Size Mode=::NewEnumerator3
    Uniform Sprite Size=200
    UsePositionOffset=false
    Write Material Random=true
   ) {module=InitializeParticle001}]
  particleUpdateScript [
   &ParticleState()
   &InheritVelocity(Inherited Velocity Speed Limit=5, Speed Limit=true, UseSpeedThreshold=false)
   &SolveForcesAndVelocity(Clamp Velocity=false, Limit Acceleration=false)
   &ScaleSpriteSize(
    Scale Sprite Size Mode=::NewEnumerator0
    Uniform Scale Factor=/Game/Effects/Niagara/DynamicInputs/Math/FN_Fit01(
     Clamp:NiagaraBool=true
     Value:NiagaraFloat=&FloatFromCurve(FloatCurve=DataInterface() {propertyValues {Curve {keys [<at 0 0 RCIM_Cubic>, {arriveTangent=1.2499, interpMode=RCIM_Cubic, leaveTangent=1.2499, tangentMode=RCTM_User, time=1, value=1}]}}})
     DestMin:NiagaraFloat=1
     DestMax:NiagaraFloat=2.5
    )
   )
   &ScaleSpriteSize(
    Scale Sprite Size Mode=::NewEnumerator0
    Uniform Scale Factor=&Waveform(
     Global Phase=&ReturnNormalizedExecIndex()
     "Waveform [1]"=::NewEnumerator0
     "[1] Frequency"=5
     "[1] Amplitude Min/Max"=<0.5 1>
    )
   ) {enabled=false, module=ScaleSpriteSize001}
   &ScaleColor(
    Scale Alpha=&FloatFromCurve(FloatCurve=DataInterface() {propertyValues {Curve {keys [
         {arriveTangent=-1.9262, interpMode=RCIM_Cubic, leaveTangent=-1.9262, tangentMode=RCTM_User, time=0, value=1}
         {tangentMode=RCTM_User, time=1, value=0}
        ]}}})
    ScaleA=true
    ScaleRGB=false
   )
   &ScaleColor(
    Scale Alpha=&FloatFromCurve(
     FloatCurve=DataInterface() {propertyValues {Curve {keys [
         {arriveTangent=-1.9262, interpMode=RCIM_Cubic, leaveTangent=-1.9262, tangentMode=RCTM_User, time=0, value=1}
         {tangentMode=RCTM_User, time=1, value=0}
        ]}}}
     CurveIndex=&RemapRange(Input Value To Remap=$Emitter.Age, Input Min=0, Input Max=$System.GlobalLifetime, Output Min=0, Output Max=1)
    )
    ScaleA=true
    ScaleRGB=false
   ) {enabled=false, module=ScaleColor001}
   /SetVariables(Particles.SpriteFacingVel:Vector3f=&NormalizeVector(Vector To Normalize=$Particles.Velocity, Safe Normalize=true, Safe Fallback Threshold=0, Safe Fallback Vector=<0 0 0.001>)) {isSetParameters=true}
   &DynamicMaterialParameters(
    Index 0 Param 1=&FN_Fit01(
     Clamp:NiagaraBool=true
     Value:NiagaraFloat=&FloatFromCurve(FloatCurve=DataInterface() {propertyValues {Curve {keys [{arriveTangent=-1.8464, interpMode=RCIM_Cubic, leaveTangent=-1.8464, tangentMode=RCTM_User, time=0, value=1}, <at 1 0 RCIM_Cubic>]}}})
     DestMin:NiagaraFloat=1
     DestMax:NiagaraFloat=1
    )
    Index 0 Param 2=&FN_Fit01(
     Clamp:NiagaraBool=true
     Value:NiagaraFloat=&FloatFromCurve(FloatCurve=DataInterface() {propertyValues {Curve {keys [{arriveTangent=-1.8464, interpMode=RCIM_Cubic, leaveTangent=-1.8464, tangentMode=RCTM_User, time=0, value=1}, <at 1 0 RCIM_Cubic>]}}})
     DestMin:NiagaraFloat=0.25
     DestMax:NiagaraFloat=0.5
    )
    Index 0 Param 3=3
    Module.Index0Param0Write=true
    Module.Index0Param1Write=true
    Module.Index0Param2Write=true
    Module.Index0Param3Write=true
   )
   &ScaleColor(
    Scale Alpha=&FloatFromCurve(
     FloatCurve=DataInterface() {propertyValues {Curve {keys [{tangentMode=RCTM_User, time=0, value=0}, {tangentMode=RCTM_User, time=0.5, value=1}]}}}
     CurveIndex=$System.Timeline.NormalizedTime
    )
    ScaleA=true
    ScaleRGB=true
   ) {module=ScaleColor002}
  ]
  rendererClasses [&NiagaraSpriteRendererProperties]
  renderers [
   {rendererClass=&NiagaraSpriteRendererProperties}
   {rendererClass=&NiagaraSpriteRendererProperties, rendererIndex=1}
  ]
 }
 PressureFront {
  emitterUpdateScript [
   &EmitterState(Inactive Response=::NewEnumerator2, Life Cycle Mode=::NewEnumerator1, UseLoopDelay=false)
   &SpawnBurst_Instantaneous(Use Spawn Probability=false, UseLoopCountLimit=true)
  ]
  particleSpawnScript [
   &InitializeParticle(Color=<rgba 1.6667 0 10 5>, Mesh Scale=<10 1 1>, Mesh Scale Mode=::NewEnumerator4, UsePositionOffset=false, Write Material Random=true)
   &InitializeParticle(Position Offset=<-100 0 0>, UsePositionOffset=true, Write Material Random=true) {enabled=false, module=InitializeParticle001}
   &InitialMeshOrientation(
    Mesh Orientation Mode=::NewEnumerator2
    Orientation Vector=&NormalizeVector(Vector To Normalize=$Engine.Owner.Velocity, Safe Normalize=true, Safe Fallback Threshold=0, Safe Fallback Vector=<0 0 0.001>)
    Use Rotation Vector=false
   ) {enabled=false}
  ]
  particleUpdateScript [
   &ParticleState(Kill Particles When Lifetime Has Elapsed=false, Let Infinitely Lived Particles Die When Emitter Deactivates=true)
   &UpdateMeshOrientation(
    Facing Direction=&NormalizeVector(Vector To Normalize=$Engine.Owner.Velocity, Safe Normalize=true, Safe Fallback Threshold=0, Safe Fallback Vector=<0 0 0.001>)
    Facing Interpolation=false
    Orientation Method=::NewEnumerator1
    Override Quat=false
    Stabilization Mode:ENiagara_OrientVectorCount=::NewEnumerator0
    TransformFacing=false
   ) {module=UpdateMeshOrientation001}
   &UpdateMeshOrientation(Orientation Method=::NewEnumerator3, Particle Velocity:Vector3f=$Engine.Owner.Velocity) {enabled=false}
   &DynamicMaterialParameters(
    Index 0 Param 1=1.5
    Index 0 Param 2=3
    Index 0 Param 3=1
    Index 0 Param 4=1
    Module.Param0WriteEnabled=true
    Module.Param1WriteEnabled=true
    Module.Param2WriteEnabled=true
    Module.Param3WriteEnabled=true
   )
   &ScaleMeshSize(Scale Factor=&VectorFromCurve(VectorCurve=DataInterface() {propertyValues {
       XCurve {keys [
         {arriveTangent=6.9057, interpMode=RCIM_Cubic, leaveTangent=6.9057, tangentMode=RCTM_User, time=0, value=0.5}
         {arriveTangent=-0.0536, interpMode=RCIM_Cubic, leaveTangent=-0.0536, tangentMode=RCTM_User, time=0.2, value=1}
         {arriveTangent=-0.0405, interpMode=RCIM_Cubic, leaveTangent=-0.0405, tangentMode=RCTM_User, time=1, value=1}
        ]}
       YCurve {keys [<at 0 1>,<at 1 1>]}
       ZCurve {keys [<at 0 1>,<at 1 1>]}
      }})) {enabled=false}
   &ScaleMeshSize(Scale Factor=&Lerp_Vector(A=<0.25 0.5 0.5>, B=<1 1 1>, Alpha=$System.Timeline.NormalizedTime)) {module=ScaleMeshSize001}
   &ScaleColor(
    Scale Alpha=&FloatFromCurve(
     FloatCurve=DataInterface() {propertyValues {Curve {keys [<at 0.2 0 RCIM_Cubic>, {tangentMode=RCTM_User, time=1, value=1}]}}}
     CurveIndex=$System.Timeline.NormalizedTime
     Scale Curve=0.5
    )
    ScaleA=true
    ScaleRGB=true
   )
  ]
  rendererClasses [&NiagaraMeshRendererProperties]
  renderers [
   {rendererClass=&NiagaraMeshRendererProperties, rendererIndex=0}
   {rendererClass=&NiagaraMeshRendererProperties, rendererIndex=1}
  ]
 }
}
summary {
 systemName=NS_TwinStretchCorn_Prj_Trail
 emitters [
  {emitterName=BulletTimeSprite, rendererClasses [&NiagaraSpriteRendererProperties]}
  {emitterName=PressureFront, rendererClasses [&NiagaraMeshRendererProperties]}
 ]
 userVariables []
}
